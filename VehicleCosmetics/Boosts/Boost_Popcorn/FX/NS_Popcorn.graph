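# Niagara vehicle boost cosmetic that streams tumbling popcorn particles backward from the car.
# One emitter with mesh and sprite renderers continuously spawns in a squashed sphere behind a DTCarOffset-adjusted position, with dimmer color in frontend preview.
# Particles follow source movement, gain randomized downward gravity over their lifetime, and scale up quickly, then shrink away.
# Painted variants, socket-based placement and game-mode velocity modules are present but disabled, leaving a fixed backward velocity.
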NiagaraSystem /VehicleCosmetics/Boosts/Boost_Popcorn/FX/NS_Popcorn
dependencies {
 usedDataInterfaces [
  &NiagaraDataInterfaceArrayColor
  &NiagaraDataInterfaceArrayFloat3
  &NiagaraDataInterfaceCurve
  &NiagaraDataInterfaceSkeletalMesh
 ]
 usedDynamicInputs []
 usedModules [
  &self:SystemSpawnScript.SystemScriptSource.SystemScriptGraph.NiagaraNodeAssignment_0.SetVariables_#1
  &SystemState
 ]
}
emitters {NE_VC_Popcorn {
  emitterUpdateScript [
   &EmitterState()
   &VC_FilteredSocketCount() {enabled=false}
   &SpawnRate(SpawnRate=100, Use Spawn Probability=false)
  ]
  particleSpawnScript [
   &VC_PaintedVariantsOverride(
    Black=<rgba 0 0 0 1>
    BurntSienna=<rgba 0.24 0.0552 0.0384 1>
    Cobalt=<rgba 0.085 0.1487 0.85 1>
    Crimson=<rgba 0.53 0.0159 0.0159 1>
    DefaultColor=&SelectLinearColorFromArray(
     Array Sampling Mode=::NewEnumerator0
     Direct Array Index=?
     Color Selection Array=DataInterface() {propertyValues {ColorData [<rgba 0.7865 0.127 0.7807 1>, <rgba 0.127 0.7039 0.7865 1>, <rgba 0.786 0.786 0.786 1>, <rgba 0.7865 0.127 0.7807 1>]}}
     Random Seed=0
    )
    ForestGreen=<rgba 0.0135 0.45 0.0135 1>
    Grey=<rgba 0.27 0.27 0.27 1>
    Lime=<rgba 0.5628 0.84 0.0084 1>
    Orange=<rgba 0.74 0.3049 0.0148 1>
    Pink=<rgba 0.85 0.255 0.8004 1>
    Purple=<rgba 0.21 0.02 0.4 1>
    R1 Black:LinearColor=<rgba 0.776 0.832 1 1>
    R1 BurntSienna:LinearColor=<rgba 0.776 0.832 1 1>
    R1 Cobalt:LinearColor=<rgba 1 1 1 1>
    R1 Crimson:LinearColor=<rgba 0.8073 0.8555 1 1>
    R1 ForestGreen:LinearColor=<rgba 1 1 1 1>
    R1 Grey:LinearColor=<rgba 1 1 1 1>
    R1 Lime:LinearColor=<rgba 1 1 1 1>
    R1 Orange:LinearColor=<rgba 0.8646 0.8984 1 1>
    R1 Pink:LinearColor=<rgba 0.8646 0.8984 1 1>
    R1 Purple:LinearColor=<rgba 1 1 1 1>
    R1 Saffron:LinearColor=<rgba 1 1 1 1>
    R1 SkyBlue:LinearColor=<rgba 1 1 1 1>
    R1 TitaniumWhite:LinearColor=<rgba 0.8646 0.8984 1 1>
    Saffron=<rgba 0.7 0.7 0.014 1>
    SkyBlue=<rgba 0.098 0.4692 0.7 1>
    TitaniumWhite=<rgba 0.9 0.9 0.9 1>
    Use R1 Painted Overrides=true
   ) {enabled=false}
   &InitializeParticle(
    Color=&MakeCustomLinearColorFromBool(Bool=$User.bIsFrontendPreview, True Color=<rgba 0.8 0.7922 0.6708 1>, False Color=<rgba 1.5 1.4853 1.2578 1>)
    Lifetime Max=1.1
    Lifetime Min=0.7
    Lifetime Mode=::NewEnumerator1
    Mesh Scale Mode=::NewEnumerator1
    Mesh Uniform Scale Max=0.2
    Mesh Uniform Scale Min=0.1
    Position Offset=<-150 0 20>
    Position Offset Coordinate Space=::Local
    UsePositionOffset=true
    Write Material Random=true
   )
   &PartitionParticles(Number Of Partitions=$Emitter.PortCount) {enabled=false}
   &SkeletalMeshLocation(
    Bone / Socket Index=$Particles.PartitionParticles.Partition
    Bone / Socket Sampling Mode=::NewEnumerator5
    MaskSpawnGroup=false
    Mesh Sampling Type=::NewEnumerator6
    Skeletal Mesh=$System.SkeletalMesh
   ) {enabled=false}
   &ShapeLocation(
    MaskSpawnGroup=false
    Non Uniform Scale=<1 1 0.7>
    Offset=&AddVectorToPosition(Position=<5 0 10>, Vector=$User.DTCarOffset)
    Offset Coordinate Space=::Local
    Offset Mode=::NewEnumerator0
    Override Seed=false
    Sphere Radius=60
    UseOwnerScale=true
   )
   &InitialMeshOrientation(Use Rotation Vector=false)
   &AddVelocity(Velocity=&MergeHelper_VectorSwitch(
     0=&SelectVectorFromArray(
      Direct Array Index=&SpawnGroup()
      Vector Selection Array=DataInterface() {propertyValues {FloatData [<-1400 0 0>, <-1300 0 0>, <-1500 0 0>]}}
     )
     1=&SelectVectorFromArray(
      Direct Array Index=&SpawnGroup()
      Vector Selection Array=DataInterface() {propertyValues {FloatData [<-1100 0 0>, <-1000 0 0>, <-1200 0 0>]}}
     )
     2=&SelectVectorFromArray(
      Direct Array Index=&SpawnGroup()
      Vector Selection Array=DataInterface() {propertyValues {FloatData [<-1000 0 0>, <-900 0 0>, <-1100 0 0>]}}
     )
     3=?
     Index=$User.GameModeID
     Num=::NewEnumerator1
    )) {enabled=false}
   &AddVelocity(Velocity=<-1200 0 0>) {module=AddVelocity001}
  ]
  particleUpdateScript [
   &ParticleState()
   &InheritSourceMovement(ApplyForce=false, ApplyPosition=true, ApplyPositionFalloff=false)
   &UpdateMeshOrientation(
    Rotation Rate=&RandomRangeFloat(Minimum=0.6, Maximum=1.2)
    Rotation Vector=&RandomVector(Vector Scale=1)
   )
   &GravityForce(Gravity=&MakeVector(
     X=0
     Y=0
     Z=&FloatFromCurve(
      FloatCurve=DataInterface() {propertyValues {Curve {keys [<at 0.2 0 RCIM_Cubic>, {arriveTangent=-2026.4426, interpMode=RCIM_Cubic, leaveTangent=-2026.4426, tangentMode=RCTM_User, time=1, value=-500}]}}}
      Scale Curve=&RandomRangeFloat(Minimum=0, Maximum=10)
     )
    ))
   &SolveForcesAndVelocity(Clamp Velocity=false, Limit Acceleration=false)
   &ScaleMeshSize(Scale Factor=&VectorFromFloat(Value=&FloatFromCurve(FloatCurve=DataInterface() {propertyValues {Curve {keys [
          <at 0 0.05>
          <at 0.12 1>
          {arriveTangent=-1.4286, arriveTangentWeight=0.3202, leaveTangent=-1.4286, leaveTangentWeight=0.1067, time=0.82, value=1}
          {arriveTangent=-8.2212, arriveTangentWeight=0.05, leaveTangent=-8.2212, time=1, value=0}
         ]}}})))
  ]
  rendererClasses [&NiagaraMeshRendererProperties, &NiagaraSpriteRendererProperties]
  renderers [
   {rendererClass=&NiagaraMeshRendererProperties, rendererIndex=0}
   {rendererClass=&NiagaraSpriteRendererProperties, rendererIndex=1}
  ]
 }}
summary {
 systemName=NS_Popcorn
 emitters [{emitterName=NE_VC_Popcorn, rendererClasses [&NiagaraMeshRendererProperties, &NiagaraSpriteRendererProperties]}]
 userVariables [
  User.bIsFrontendPreview {type=NiagaraBool}
  User.bIsPainted {type=NiagaraBool}
  User.ColorVariantID {type=NiagaraInt32}
  User.DTCarOffset {type=Vector3f, description="", defaultValue {struct=Vector3f, value=<0 0 0>}}
  User.GameModeID {type=NiagaraInt32, defaultValue=2}
  User.PaintedColor {type=LinearColor, description="", defaultValue {struct=LinearColor, value=<rgba 1 1 1 1>}}
  User.PaintedScalar {type=NiagaraFloat, defaultValue=1, description=""}
  User.PortCount {type=NiagaraInt32, defaultValue=2}
 ]
}
